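# Translucent, unlit, two-sided material for the No Weapon Zone toy's boundary, drawing animated glowing bands on its sides.
# World position, projected by the rounded normal, forms vertical bands that narrow with camera distance and are modulated by stepped sweeps panning at randomized speeds.
# Fresnel and an actor-relative height gradient add glow, with horizontal faces masked out and backfaces dimmed.
# The combined mask drives color and opacity, with near-camera and depth fades at medium and high quality.
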Material /Game/Abilities/Toys/NoWeaponZone/M_NoWeaponZone_Bounds
graph {
 properties {
  bEnableResponsiveAA=true
  blendMode=BLEND_Translucent
  pixelDepthOffsetMode=PDOM_Legacy
  shadingModel=MSM_Unlit
  twoSided=true
  usedShadingModels=Unlit
 }
 nodes [
  n1 = MaterialFunctionCall(Float3=Divide(A=WorldPosition().XYZ, B=ScalarParameter() {defaultValue=6, parameterName=WorldScale})) {materialFunction=&BreakOutFloat3Components}
  n2 = VertexNormalWS()
  n3 = MaterialFunctionCall(Float3=Abs(None=Round(None=@n2))) {materialFunction=&BreakOutFloat3Components}
  n4 = LinearInterpolate(
   A=LinearInterpolate(A=AppendVector(A=@n1.G, B=@n1.G), B=AppendVector(A=@n1.R, B=@n1.R), Alpha=@n3.G)
   B=AppendVector(A=@n1.R, B=@n1.R)
   Alpha=@n3.B
  )
  n5 = Add(A=Distance(A=CameraPositionWS(), B=WorldPosition().XYZ), B=ScalarParameter() {defaultValue=-100, parameterName=DistanceBias})
  n6 = ScalarParameter() {defaultValue=1600, parameterName=FadeScale}
  n7 = OneMinus(None=Saturate(None=Divide(A=@n5, B=@n6)))
  n8 = ScalarParameter() {defaultValue=0.22, parameterName=DotsRadius}
  n9 = Power(
   Base=SphereMask(
    A=Frac(None=@n4)
    B=Constant() {r=0.5}
    Radius=Multiply(A=@n7, B=@n8)
    Hardness=ScalarParameter() {defaultValue=0.2, parameterName=DotsHardness}
   )
   Exp=ScalarParameter() {defaultValue=3, parameterName=DotsExponent}
  )
  n10 = Constant() {r=1}
  n11 = Divide(A=MaterialFunctionCall() {materialFunction=/Engine/Functions/Engine_MaterialFunctions02/WorldPositionOffset/ObjectScale}.Scale Z) {constB=384}
  n12 = MaterialFunctionCall!(
   PrimaryTextureCoordinates=AppendVector(
    A=MaterialFunctionCall(Float2=@n4) {materialFunction=&BreakOutFloat2Components}.R
    B=Constant() {r=1}
   )
   2ndOrderTextureCoordinates
   RandomSeed=ScalarParameter() {defaultValue=3, parameterName=RandomSeed}
   NoiseDensityBias
   Bypass Noise Influence Crosstalk
  ) {materialFunction=/Game/Effects/Fort_Effects/Materials/_Functions/MF_RandomUVSample}
  n13 = Multiply(A=@n12.NoiseWithRange, B=ScalarParameter() {defaultValue=-0.13, parameterName=PanSpeed})
  n14 = Panner(
   Coordinate=Multiply(A=TextureCoordinate(), B=AppendVector(A=@n10, B=@n11))
   Speed=AppendVector(A=Constant(), B=Add(A=@n13) {constB=-0.04, desc=Make sure we don't get any 0s})
  ) {speedY=-0.1}
  n15 = SphereMask(
   A=Frac(None=MaterialFunctionCall(Float2=@n14) {materialFunction=&BreakOutFloat2Components}.G)
   B=Constant() {r=1}
   Radius=ScalarParameter() {defaultValue=0.4, parameterName=SweepGradientRadius}
   Hardness=Constant()
  )
  n16 = ScalarParameter() {defaultValue=38.4, parameterName=SweepSteps}
  n17 = Divide(A=Floor(None=Multiply(A=@n15, B=@n16)), B=@n16)
  n18 = LinearInterpolate(A=Multiply(A=ScalarParameter() {defaultValue=0.2, parameterName=DotsBackfaceOpacity}, B=@n17), B=@n17, Alpha=Saturate(None=TwoSidedSign()))
  n19 = Add(
   A=Multiply(A=@n18, B=ScalarParameter() {defaultValue=0.015, parameterName=SweepBlocksScale})
   B=MaterialFunctionCall(
    Brightness=ScalarParameter() {defaultValue=0.11, parameterName=FresnelBrightness}
    Exponent=ScalarParameter() {defaultValue=4, parameterName=FresnelExponent}
   ) {materialFunction=&MF_Fresnel_FX}.Fresnel Inverted Double Sided
  )
  n20 = WorldPosition()
  n21 = ActorPositionWS()
  n22 = Saturate(None=Divide(A=Subtract(A=@n20.XYZ, B=@n21), B=ScalarParameter() {defaultValue=150, parameterName=HeightGradientDivisor}))
  n23 = Multiply(
   A=MaterialFunctionCall(Float3=OneMinus(None=@n22)) {materialFunction=&BreakOutFloat3Components}.B
   B=ScalarParameter() {defaultValue=0.06, parameterName=HeightGradientScale}
  )
  n24 = Multiply(A=Add(A=Multiply(A=@n9, B=@n18), B=Add(A=@n19, B=@n23)), B=OneMinus(None=@n3.B))
  LinearInterpolate(
   A=VectorParameter() {parameterName=ColorLow, defaultValue=<rgba 0.0628 0.1086 0.5 1>}.RGB
   B=VectorParameter() {parameterName=ColorHigh, defaultValue=<rgba 2.7786 15.0862 20 1>}.RGB
   Alpha=@n24
  )
  n25 = Time()
  n26 = ScalarParameter() {defaultValue=10, parameterName=PanTimeScale}
  Multiply(A=@n12.NoiseWithRange, B=Ceil(None=Multiply(A=@n25, B=@n26)))
  n27 = Add(A=ScalarParameter() {defaultValue=0.2, parameterName=Opacity}, B=@n24)
  n28 = ScalarParameter() {defaultValue=200, parameterName="Fade_Start_Distance (S)"}
  n29 = DepthFade(
   Opacity=Multiply(A=@n27, B=MaterialFunctionCall(Distance=@n28) {materialFunction=&Near_Camera_Fade}.Result)
   FadeDistance=ScalarParameter() {defaultValue=20, parameterName=FadeDistance}
  )
  QualitySwitch(Default=@n27, Low=@n27, High=@n29, Medium=@n29)
 ]
}
