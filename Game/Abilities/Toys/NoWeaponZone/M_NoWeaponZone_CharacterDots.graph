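# Translucent unlit particle-sprite material for the No Weapon Zone toy, drawing a color-ramped dot overlay, apparently over characters.
# Scene-texture world normals, in camera space, warp the dot grid and drive a fresnel term, while camera distance shrinks the dots.
# Randomized per-column panning creates stepped sweeping bands that modulate the dots.
# A circular mask, dynamic Sweep parameter and depth fades shape opacity; non-deferred shading paths get zero constants.
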Material /Game/Abilities/Toys/NoWeaponZone/M_NoWeaponZone_CharacterDots
graph {
 properties {
  bEnableResponsiveAA=true
  bUsedWithParticleSprites=true
  blendMode=BLEND_Translucent
  pixelDepthOffsetMode=PDOM_Legacy
  shadingModel=MSM_Unlit
  usedShadingModels=Unlit
 }
 nodes [
  ConstantBiasScale(None=Sine(None=Time()) {period=3})
  n1 = Constant()
  n2 = VectorParameter() {parameterName=ColorLow, defaultValue=<rgba 0.088 0.152 0.7 1>}
  n3 = VectorParameter() {parameterName=ColorHigh, defaultValue=<rgba 5.5572 30.1725 40 1>}
  n4 = SceneTexture() {sceneTextureId=PPI_WorldNormal}
  n5 = Multiply(
   A=Transform(None=@n4.Color) {transformSourceType=TRANSFORMSOURCE_World, transformType=TRANSFORM_Camera}
   B=ScalarParameter() {defaultValue=0.35, parameterName=NormalScale}
  )
  n6 = TextureCoordinate()
  n7 = ScalarParameter() {defaultValue=80, parameterName=DotsRepeat}
  n8 = Frac(None=Add(A=ComponentMask(None=@n5) {g=true, r=true}, B=Multiply(A=@n6, B=@n7)))
  n9 = Constant() {r=0.5}
  n10 = CameraPositionWS()
  n11 = WorldPosition()
  n12 = Divide(
   A=Add(A=Distance(A=@n10, B=@n11.XYZ), B=ScalarParameter() {defaultValue=-100, parameterName=DistanceBias})
   B=ScalarParameter() {defaultValue=1600, parameterName=FadeScale}
  )
  n13 = Multiply(A=OneMinus(None=Saturate(None=@n12)), B=ScalarParameter() {defaultValue=0.5, parameterName=DotsRadius})
  n14 = ScalarParameter() {defaultValue=0.2, parameterName=DotsHardness}
  n15 = MaterialFunctionCall(Float2=Multiply(A=TextureCoordinate(), B=ScalarParameter() {defaultValue=80, parameterName=LinesRepeat}) {constB=10}) {materialFunction=&BreakOutFloat2Components}
  n16 = Constant() {r=1}
  n17 = Multiply(
   A=MaterialFunctionCall!(
    PrimaryTextureCoordinates=AppendVector(A=@n15.R, B=@n16)
    2ndOrderTextureCoordinates
    RandomSeed=ScalarParameter() {defaultValue=3, parameterName=RandomSeed}
    NoiseDensityBias
    Bypass Noise Influence Crosstalk
   ) {materialFunction=/Game/Effects/Fort_Effects/Materials/_Functions/MF_RandomUVSample}.NoiseWithRange
   B=ScalarParameter() {defaultValue=1.65, parameterName=PanSpeed}
  )
  n18 = Panner(Coordinate=TextureCoordinate(), Speed=AppendVector(A=Constant(), B=Add(A=@n17) {constB=0.1, desc=Make sure we don't get any 0s})) {speedY=-0.1}
  n19 = SphereMask(
   A=Frac(None=MaterialFunctionCall(Float2=@n18) {materialFunction=&BreakOutFloat2Components}.G)
   B=Constant()
   Radius=ScalarParameter() {defaultValue=0.4, parameterName=SweepGradientRadius}
   Hardness=Constant()
  )
  n20 = ScalarParameter() {defaultValue=38.4, parameterName=SweepSteps}
  n21 = Divide(A=Floor(None=Multiply(A=@n19, B=@n20)), B=@n20)
  n22 = Multiply(
   A=Power(Base=SphereMask(A=@n8, B=@n9, Radius=@n13, Hardness=@n14), Exp=ScalarParameter() {defaultValue=4, parameterName=DotsExponent})
   B=@n21
  )
  n23 = Add(
   A=MaterialFunctionCall(
    Brightness=ScalarParameter() {defaultValue=0.8, parameterName=FresnelBrightness}
    Exponent=ScalarParameter() {defaultValue=4, parameterName=FresnelExponent}
    Vertex / Pixel Normal=ComponentMask(None=@n4.Color) {b=true, g=true, r=true}
   ) {materialFunction=&MF_Fresnel_FX}.Fresnel Inverted
   B=@n22
  )
  MaterialFunctionCall(Emissive=ShadingPathSwitch(Default=@n1, Deferred=LinearInterpolate(A=@n2.RGB, B=@n3.RGB, Alpha=@n23), Forward=@n1, Mobile=@n1)) {materialFunction=&MF_FX_EmissiveAdjustments}
  n24 = Constant()
  n25 = ScalarParameter() {defaultValue=0.7, parameterName=Opacity}
  n26 = Add(A=@n22, B=Multiply(A=@n21, B=ScalarParameter() {defaultValue=0.015, parameterName=SweepBlocksScale}))
  n27 = TextureCoordinate()
  n28 = Constant() {r=0.5}
  n29 = ScalarParameter() {defaultValue=0.5, parameterName=FinalMaskHardness}
  n30 = ScalarParameter() {defaultValue=80, parameterName=Distance Divisor}
  n31 = Multiply(
   A=Multiply(A=Add(A=@n25, B=@n26), B=SphereMask(A=@n27, B=@n28, Radius=@n28, Hardness=@n29))
   B=Power(Base=MaterialFunctionCall!(Distance Bias, Distance Divisor=@n30, Scene Depth, Pixel Depth) {materialFunction=/Game/Effects/Fort_Effects/Materials/_Functions/MF_DistanceBehindTranslucency_01}.Normalized Inverted) {constExponent=0.2}
  )
  n32 = TextureCoordinate()
  n33 = SphereMask(
   A=Add(A=ComponentMask(None=@n32) {g=true}, B=DynamicParameter() {defaultValue=<rgba 0 1 1 1>, paramNames [Sweep,Param2,Param3,Param4]}.Sweep)
   B=Constant() {r=1}
   Radius=ScalarParameter() {defaultValue=0.3, parameterName=SweepRadius}
   Hardness=Constant()
  )
  ShadingPathSwitch(
   Default=@n24
   Deferred=DepthFade(Opacity=Multiply(A=@n31, B=@n33), FadeDistance=ScalarParameter() {defaultValue=3, parameterName=FadeDistance})
   Forward=@n24
   Mobile=@n24
  )
 ]
}
