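# Additive, unlit sprite-particle material that draws a soft round glowing dot for the NoWeaponZone toy ability effect.
# A zero-hardness SphereMask centered on the UVs makes the dot procedurally without textures, and an Exponent parameter tightens its falloff.
# The mask multiplies particle color RGB and feeds MF_FX_EmissiveAdjustments as emissive.
# Responsive AA is enabled.
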Material /Game/Abilities/Toys/NoWeaponZone/M_NoWeaponZone_Dot
graph {
 properties {
  bEnableResponsiveAA=true
  bUsedWithParticleSprites=true
  blendMode=BLEND_Additive
  pixelDepthOffsetMode=PDOM_Legacy
  shadingModel=MSM_Unlit
  usedShadingModels=Unlit
 }
 nodes [
  n1 = Constant() {r=0.5}
  n2 = SphereMask(A=TextureCoordinate(), B=@n1, Radius=@n1, Hardness=Constant())
  n3 = ScalarParameter() {defaultValue=8, parameterName=Exponent}
  MaterialFunctionCall(Emissive=Multiply(A=ParticleColor().RGB, B=Power(Base=@n2, Exp=@n3))) {materialFunction=&MF_FX_EmissiveAdjustments}
 ]
}
