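# Additive, unlit, two-sided mesh particle material for the glyphs of the No Weapon Zone ability, glowing in an adjustable edge color.
# Opacity blends dim inverted Fresnel toward solid using vertex red and a U-coordinate sphere mask, then applies vertex green, depth fade and particle alpha.
# World-position-based curl noise, scaled by particle relative time, drives a vertex offset that increasingly distorts the glyph over its lifetime.
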Material /Game/Abilities/Toys/NoWeaponZone/M_NoWeaponZone_Glyph
graph {
 properties {
  bEnableResponsiveAA=true
  bUsedWithMeshParticles=true
  blendMode=BLEND_Additive
  pixelDepthOffsetMode=PDOM_Legacy
  previewMesh=&/Sm_NoWeaponZone_GlyphA
  shadingModel=MSM_Unlit
  twoSided=true
  usedShadingModels=Unlit
 }
 nodes [
  MaterialFunctionCall(Emissive=VectorParameter() {parameterName=EdgeColor, defaultValue=<rgba 1.0053 3.6292 10 0>}.RGB) {materialFunction=&MF_FX_EmissiveAdjustments}
  n1 = VertexColor()
  n2 = Constant() {r=0.5}
  n3 = SphereMask(A=ComponentMask(None=TextureCoordinate()) {r=true}, B=@n2, Radius=@n2, Hardness=Constant())
  n4 = LinearInterpolate(
   A=MaterialFunctionCall(Brightness=ScalarParameter() {defaultValue=0.05, parameterName=FresnelBrightness}) {materialFunction=&MF_Fresnel_FX}.Fresnel Inverted Double Sided
   Alpha=Multiply(A=@n1.R, B=Power(Base=@n3) {constExponent=4.5})
  ) {constA=0.1}
  Multiply(
   A=DepthFade(Opacity=Multiply(A=@n4, B=@n1.G), FadeDistance=ScalarParameter() {defaultValue=5, parameterName=FadeDistance})
   B=ParticleColor().A
  )
  n5 = Divide(A=WorldPosition().XYZ, B=ScalarParameter() {defaultValue=40, parameterName=CurlScale})
  Multiply(
   A=Multiply(A=VectorNoise(World Position=@n5) {noiseFunction=VNF_CurlALU}, B=ParticleRelativeTime())
   B=ScalarParameter() {defaultValue=2, parameterName=wpoAmount}
  )
 ]
}
